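; Parent spawner actor that spawns an Umbramolt Fraternal AI pawn from a row in the CH7S1 Lager boss event table.
; On the server, once playlist data is ready, it records the instigating player and asks the Living World Manager to spawn at its location and yaw.
; It feeds any valid activating player to the bot controller as a seen threat.
; It destroys itself when that pawn dies or the spawned-pawn cast fails.
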(event EventBeginPlay
  (bind (_returnvalue_roll _returnvalue_pitch _returnvalue_yaw) (Math|Rotator|BreakRotator (Transformation|GetActorRotation)))
  (Networking|SwitchHasAuthority
    (:Authority
      (Class|FortAsyncActionCurrentPlaylistReady|PlaylistDataReadyAsync
        (:OnReady
          (bind _asfort_player_pawn_athena (Utilities|Casting|CastToFortPlayerPawnAthena (Game|GetInstigator))
            (:then
              (Variables|Default|SetActivatingPlayer _asfort_player_pawn_athena))
            (:CastFailed
              _))
          (Class|FortAthenaLivingWorldManager|TrySpawnEvent (Class|FortAthenaLivingWorldManager|GetLivingWorldManager) (Utilities|Struct|MakeDataTableRowHandle "/Lager/Lager/EventTables/Lager_EventTable_Boss_CH7S1.Lager_EventTable_Boss_CH7S1" (Variables|Default|GetEventRowToSpawn)) (Math|Transform|MakeTransform (Transformation|GetActorLocation) (Math|Rotator|MakeRotator 0.0 0.0 _returnvalue_yaw)) (delegate Custom|FraternalActorSpawned)))))))

(event Custom|FraternalActorSpawned
  (params (SpawnedActors :type Array%20of%20Actor%20Object%20References) (bSuccess :type Boolean))
  (bind _asbp_player_pawn_non_participant (Utilities|Casting|CastToBP_PlayerPawn_NonParticipant (Utilities|Array|Get(acopy) SpawnedActors))
    (:then
      (bind _output_get (Variables|Default|SetSpawnedFraternalPawn _asbp_player_pawn_non_participant))
      (EventDispatchers|BindEventtoOnDied _output_get (delegate Custom|FraternalPawnDied))
      (Utilities|IsValid (Variables|Default|GetActivatingPlayer)
        (:Is%20Valid
          (bind _asfort_athena_aibot_controller (Utilities|Casting|CastToFortAthenaAIBotController (Fortnite|GetControllerfromActor (Variables|Default|GetSpawnedFraternalPawn))))
          (Class|FortAthenaAIBotController|AddTargetInfo _asfort_athena_aibot_controller (Variables|Default|GetActivatingPlayer) true "Threat_Seeing" 5.0 1e+04))))
    (:CastFailed
      (Actor|DestroyActor))))

(event Custom|FraternalPawnDied
  (params (DamagedActor :type Actor%20Object%20Reference) (Damage :type Float%20%28single-precision%29) (InstigatedBy :type Controller%20Object%20Reference) (DamageCauser :type Actor%20Object%20Reference) (HitLocation :type Vector) (FHitComponent :type Primitive%20Component%20Object%20Reference) (BoneName :type Name) (Momentum :type Vector))
  (Actor|DestroyActor))

(fn ConstructionScript)
